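# Editor utility list-entry widget for the Texture Review Tool, showing one GART asset report with thumbnail, name and reason.
# On item assignment it reads GartData by name via GetEditorProperty, then displays the first report, switching description and red tinting on one severity level.
# Colored channel strips appear when the long description contains Red, Green, Blue or Alpha.
# Hover tints the main border, while list selection reveals the imgHover overlay.
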EditorUtilityWidgetBlueprint /Game/NeverCook/EditorUtilities/PerfTools/TextureReviewTool/EUW_TextureReviewToolItem
graphs [
 EventGraph {nodes [
   Event PreConstruct()
   > SetVisibility(self=Get imgHover() -> (imgHover:Image Object Reference), InVisibility=Hidden)
   n1 = Event On Item Selection Changed!() -> (OutputDelegate, then, bIsSelected:bool) {is=title K2Node_Event}
   > Set IsCurrentlySelectedInListView(IsCurrentlySelectedInListView:bool=@n1.bIsSelected) -> (Output_Get:bool)
   > ValidateSelectionVisuals()
   Event OnMouseEnter()
   > ValidateSelectionVisuals()
   Event OnMouseLeave!() -> (OutputDelegate, then, MouseEvent:"Pointer Event Structure (by ref)") {is=title K2Node_Event}
   > ValidateSelectionVisuals()
   n2 = Event OnListItemObjectSet()
   > HandleListItemObjectSet!(execute, self, listObject:Object Reference=@n2.ListItemObject) -> (then)
   Event Construct()
   Event Tick()
  ]}
 HandleListItemObjectSet {nodes [
   HandleListItemObjectSet!() -> (then, listObject:Object Reference) {is=title K2Node_FunctionEntry}
   > n1 = Sequence!(execute) -> (then_0, then_1, then_2:Exec, then_3:Exec)
   > n2 = GetEditorProperty!(execute:Exec, Object:Object Reference=Get listObject() -> (listObject:Object Reference), PropertyName:Name=GartData) -> (then:Exec, PropertyValue:Struct GART Asset Report List Structure, ReturnValue:bool)
   > n3 = Branch(Condition=@n2.ReturnValue)
   > Set GartData(GartData:Struct GART Asset Report List Structure=@n2) -> (Output_Get:Struct GART Asset Report List Structure)
   Return Node(execute=@n3.else)
   SetAssetData!(
    execute=@n1.then_1
    self
    AssetData:"Asset Data Structure (by ref)"=Break Struct GART Asset Report List!(Struct_GART_AssetReportList:"Struct GART Asset Report List Structure (by ref)"=Get GartData() -> (GartData:Struct GART Asset Report List Structure)) -> (
     ReportDatas_11_#1:Array of Struct GART Report Output Data Structures
     AssetData_7_#2:Asset Data Structure
     Object_10_#3:Object Reference
    ) {is=title K2Node_BreakStruct}.AssetData_7_#2
   ) -> (then)
   n4 = Break Struct GART Asset Report List(Struct_GART_AssetReportList=Get GartData())
   n5 = Break Struct GART Report Data!(Struct_GART_ReportData:"Struct GART Report Data Structure (by ref)"=Break Struct GART Report Output Data!(Struct_GART_ReportOutputData:"Struct GART Report Output Data Structure (by ref)"="Get (a copy)"(Array:Array of Struct GART Report Output Data Structures=@n4) -> (Output:Struct GART Report Output Data Structure)) -> (ReportData_46_#4:Struct GART Report Data Structure) {is=title K2Node_BreakStruct}) -> (
    DisplayTitle_20_#5:String
    DisplayDescShort_29_#6:String
    DisplayDescLong_30_#7:String
    DisplayColorOverride_40_#8:Linear Color Structure
    Tags_25_#9:Array of Names
    Severity_46_#4:Enum_GART_ReportSeverity Enum=~
   ) {is=title K2Node_BreakStruct}
   Set l_desc(execute=@n1.then_2, l_desc:String=@n5.DisplayDescLong_30_#7) -> (Output_Get:String)
   n6 = "Equal (Enum)"(A:Enum_GART_ReportSeverity Enum=@n5.Severity_46_#4, B:Enum_GART_ReportSeverity Enum=NewEnumerator2)
   > "SetText (Text)"(
    self=Get txtReason() -> (txtReason:Text Object Reference)
    InText=Select(
     Option 0:Text="To Text (String)"(InString=@n5.DisplayDescLong_30_#7)
     Option 1:Text="To Text (String)"(InString=@n5.DisplayDescShort_29_#6)
     Index=@n6
    ) -> (ReturnValue:Text)
   )
   > SetBrushColor(
    self=Get BorderTitle() -> (BorderTitle:Border Object Reference)
    InBrushColor=Select(Option 0:Linear Color Structure=(R=0,G=0,B=0,A=0.502), Option 1:Linear Color Structure=(R=0.2667,G=0.0157,B=0.0157,A=1), Index=@n6) -> (ReturnValue:Linear Color Structure)
   )
   > SetBrushColor(
    self=Get borderContent() -> (borderContent:Border Object Reference)
    InBrushColor=Select(Option 0=(R=0,G=0,B=0,A=0), Option 1=(R=0.2118,G=0.0118,B=0.0118,A=0.251), Index=@n6)
   )
   n7 = Get l_desc() -> (l_desc:String)
   SetVisibility(
    execute=@n1.then_3
    self=Get imgRed() -> (imgRed:Image Object Reference)
    InVisibility=Select(Option 0:ESlateVisibility Enum=Collapsed, Option 1:ESlateVisibility Enum=Visible, Index=Contains(SearchIn=@n7, Substring=Red)) -> (ReturnValue:ESlateVisibility Enum)
   )
   n8 = Get l_desc()
   > SetVisibility(
    self=Get imgGreen() -> (imgGreen:Image Object Reference)
    InVisibility=Select(Option 0=Collapsed, Option 1=Visible, Index=Contains(SearchIn=@n8, Substring=Green))
   )
   n9 = Get l_desc()
   > SetVisibility(
    self=Get imgBlue() -> (imgBlue:Image Object Reference)
    InVisibility=Select(Option 0=Collapsed, Option 1=Visible, Index=Contains(SearchIn=@n9, Substring=Blue))
   )
   n10 = Get l_desc()
   > SetVisibility(
    self=Get imgAlpha() -> (imgAlpha:Image Object Reference)
    InVisibility=Select(Option 0=Collapsed, Option 1=Visible, Index=Contains(SearchIn=@n10, Substring=Alpha))
   )
  ]}
 SetAssetData {nodes [
   n1 = SetAssetData!() -> (then, AssetData:"Asset Data Structure (by ref)") {is=title K2Node_FunctionEntry}
   > SetAsset!(
    execute
    self:Asset Thumbnail Widget Object Reference=Get AssetThumbnailWidget_0() -> (AssetThumbnailWidget_0:Asset Thumbnail Widget Object Reference)
    AssetData:"Asset Data Structure (by ref)"=@n1
   ) -> (then)
   n2 = Break Asset Data(AssetData=@n1)
   n3 = Get txtMain() -> (txtMain:Text Object Reference)
   > "SetText (Text)"(self=@n3, InText="To Text (Name)"!(InName:Name=@n2.AssetName) -> (ReturnValue:Text))
   > SetToolTipText!(execute, self:Widget Object Reference=@n3, InToolTipText:"Text (by ref)"="To Text (Name)"(InName=@n2)) -> (then)
  ]}
 ValidateSelectionVisuals {nodes [
   ValidateSelectionVisuals() {is=title K2Node_FunctionEntry}
   > SetBrushColor(
    self=Get BorderMain() -> (BorderMain:Border Object Reference)
    InBrushColor=Select(Option 0:Linear Color Structure=(R=0,G=0,B=0,A=0), Option 1:Linear Color Structure=(R=1,G=1,B=1,A=0.1), Index=IsHovered()) -> (ReturnValue:Linear Color Structure)
   )
   > SetVisibility(
    self=Get imgHover()
    InVisibility=Select(
     Option 0:ESlateVisibility Enum=Hidden
     Option 1:ESlateVisibility Enum=Visible
     Index=Get IsCurrentlySelectedInListView() -> (IsCurrentlySelectedInListView:bool)
    ) -> (ReturnValue:ESlateVisibility Enum)
   )
  ]}
]
variables [
 GartData {type=Struct_GART_AssetReportList, default {object=null, assetData {}, reportDatas []}}
 Index {type=boolean}
 IsCurrentlySelectedInListView {type=boolean}
]
widgets {
 info {rootWidgetClass=&Border, widgetCount=17}
 widgets [
  BorderMain {
   widgetClassPath=&Border
   bIsVariable=true
   widget=BorderMain
   properties {verticalAlignment=VAlign_Top, brushColor=<rgba 0.0176 0.0176 0.0176 1>, padding=<ltrb 0 0 0 0>}
  }
  VerticalBox_0 {
   widgetClassPath=&VerticalBox
   parent=BorderMain
   widget=VerticalBox_0
   properties {clipping=ClipToBoundsAlways}
   slot {class=BorderSlot, properties {verticalAlignment=VAlign_Top, padding=<ltrb 0 0 0 0>}}
  }
  BorderTitle {
   widgetClassPath=&Border
   bIsVariable=true
   parent=VerticalBox_0
   widget=BorderTitle
   properties {brushColor=<rgba 0 0 0 0.5>, padding=<ltrb 2 2 2 2>}
   slot {class=VerticalBoxSlot}
  }
  txtMain {
   widgetClassPath=&TextBlock
   bIsVariable=true
   parent=BorderTitle
   widget=txtMain
   properties {
    text=Text Block
    font {
     bForceMonospaced=false
     bMaterialIsStencil=false
     fontMaterial=null
     fontObject=&Roboto
     letterSpacing=0
     monospacedWidth=1
     size=9
     skewAmount=0
     typefaceFontName=Regular
     outlineSettings {
      bApplyOutlineToDropShadows=false
      bMiteredCorners=false
      bSeparateFillAlpha=false
      outlineMaterial=null
      outlineSize=0
      outlineColor=<rgba 0 0 0 1>
     }
    }
   }
   slot {class=BorderSlot, properties {padding=<ltrb 2 2 2 2>}}
  }
  HorizontalBox_2 {
   widgetClassPath=&HorizontalBox
   parent=VerticalBox_0
   widget=HorizontalBox_2
   slot {class=VerticalBoxSlot, properties {verticalAlignment=VAlign_Top}}
  }
  VerticalBox_64 {widgetClassPath=&VerticalBox, parent=HorizontalBox_2, widget=VerticalBox_64, slot {class=HorizontalBoxSlot}}
  imgRed {
   widgetClassPath=&Image
   bIsVariable=true
   parent=VerticalBox_64
   widget=imgRed
   properties {brush {
     bIsDynamicallyLoaded=false
     drawAs=Image
     imageType=NoImage
     mirroring=NoMirror
     resourceName=None
     resourceObject=null
     tiling=NoTile
     imageSize=<4 32>
     margin=<ltrb 0 0 0 0>
     outlineSettings {
      bUseBrushTransparency=false
      roundingType=HalfHeightRadius
      width=0
      color {colorUseRule=UseColor_Specified, specifiedColor=<rgba 0 0 0 0>}
      cornerRadii=<0 0 0 0>
     }
     tintColor {colorUseRule=UseColor_Specified, specifiedColor=<rgba 1 0 0 1>}
     uVRegion {bIsValid=false, max=<0 0>, min=<0 0>}
    }}
   slot {class=VerticalBoxSlot, properties {size {sizeRule=Fill, value=1}}}
  }
  imgGreen {
   widgetClassPath=&Image
   bIsVariable=true
   parent=VerticalBox_64
   widget=imgGreen
   properties {brush {
     bIsDynamicallyLoaded=false
     drawAs=Image
     imageType=NoImage
     mirroring=NoMirror
     resourceName=None
     resourceObject=null
     tiling=NoTile
     imageSize=<4 32>
     margin=<ltrb 0 0 0 0>
     outlineSettings {
      bUseBrushTransparency=false
      roundingType=HalfHeightRadius
      width=0
      color {colorUseRule=UseColor_Specified, specifiedColor=<rgba 0 0 0 0>}
      cornerRadii=<0 0 0 0>
     }
     tintColor {colorUseRule=UseColor_Specified, specifiedColor=<rgba 0 1 0.009 1>}
     uVRegion {bIsValid=false, max=<0 0>, min=<0 0>}
    }}
   slot {class=VerticalBoxSlot, properties {size {sizeRule=Fill, value=1}}}
  }
  imgBlue {
   widgetClassPath=&Image
   bIsVariable=true
   parent=VerticalBox_64
   widget=imgBlue
   properties {brush {
     bIsDynamicallyLoaded=false
     drawAs=Image
     imageType=NoImage
     mirroring=NoMirror
     resourceName=None
     resourceObject=null
     tiling=NoTile
     imageSize=<4 32>
     margin=<ltrb 0 0 0 0>
     outlineSettings {
      bUseBrushTransparency=false
      roundingType=HalfHeightRadius
      width=0
      color {colorUseRule=UseColor_Specified, specifiedColor=<rgba 0 0 0 0>}
      cornerRadii=<0 0 0 0>
     }
     tintColor {colorUseRule=UseColor_Specified, specifiedColor=<rgba 0 0.0852 1 1>}
     uVRegion {bIsValid=false, max=<0 0>, min=<0 0>}
    }}
   slot {class=VerticalBoxSlot, properties {size {sizeRule=Fill, value=1}}}
  }
  imgAlpha {
   widgetClassPath=&Image
   bIsVariable=true
   parent=VerticalBox_64
   widget=imgAlpha
   properties {brush {
     bIsDynamicallyLoaded=false
     drawAs=Image
     imageType=NoImage
     mirroring=NoMirror
     resourceName=None
     resourceObject=null
     tiling=NoTile
     imageSize=<4 32>
     margin=<ltrb 0 0 0 0>
     outlineSettings {
      bUseBrushTransparency=false
      roundingType=HalfHeightRadius
      width=0
      color {colorUseRule=UseColor_Specified, specifiedColor=<rgba 0 0 0 0>}
      cornerRadii=<0 0 0 0>
     }
     tintColor {colorUseRule=UseColor_Specified, specifiedColor=<rgba 1 1 1 1>}
     uVRegion {bIsValid=false, max=<0 0>, min=<0 0>}
    }}
   slot {class=VerticalBoxSlot, properties {size {sizeRule=Fill, value=1}}}
  }
  borderContent {
   widgetClassPath=&Border
   bIsVariable=true
   parent=HorizontalBox_2
   widget=borderContent
   properties {brushColor=<rgba 1 1 1 0>, padding=<ltrb 0 0 0 0>}
   slot {class=HorizontalBoxSlot, properties {size {sizeRule=Fill, value=1}}}
  }
  overlay {widgetClassPath=&Overlay, parent=borderContent, widget=overlay, slot {class=BorderSlot, properties {padding=<ltrb 0 0 0 0>}}}
  imgHover {
   widgetClassPath=&Image
   bIsVariable=true
   parent=overlay
   widget=imgHover
   properties {visibility=Hidden, colorAndOpacity=<rgba 0.2806 0.5878 1 0.1>}
   slot {class=OverlaySlot, properties {horizontalAlignment=HAlign_Fill, verticalAlignment=VAlign_Fill}}
  }
  stack {
   widgetClassPath=&StackBox
   parent=overlay
   widget=stack
   slot {class=OverlaySlot, properties {horizontalAlignment=HAlign_Fill, verticalAlignment=VAlign_Fill}}
  }
  AssetThumbnailWidget_0 {
   widgetClassPath=/Script/UMGEditor.AssetThumbnailWidget
   bIsVariable=true
   parent=stack
   widget=AssetThumbnailWidget_0
   properties {thumbnailSettings {
     bAllowAssetSpecificThumbnailOverlay=false
     bAllowHintText=true
     bAllowRealTimeOnHovered=true
     bForceGenericThumbnail=false
     bOverrideAssetTypeColor=false
     colorStripOrientation=HorizontalBottomEdge
     genericThumbnailSize=24
     highlightedTextDelegate="(null).None"
     thumbnailLabel=ClassName
     assetTypeColorOverride=<rgba 0 0 0 0>
     hintColorAndOpacity=<rgba 0 0 0 0>
     padding=<ltrb 0 0 0 0>
    }}
   slot {class=StackBoxSlot, properties {padding=<ltrb 3 3 3 3>}}
  }
  txtReason {
   widgetClassPath=&TextBlock
   bIsVariable=true
   parent=stack
   widget=txtReason
   properties {
    text="🟢Text Block"
    font {
     bForceMonospaced=false
     bMaterialIsStencil=false
     fontMaterial=null
     fontObject=&Roboto
     letterSpacing=0
     monospacedWidth=1
     size=9
     skewAmount=0
     typefaceFontName=Bold
     outlineSettings {
      bApplyOutlineToDropShadows=false
      bMiteredCorners=false
      bSeparateFillAlpha=false
      outlineMaterial=null
      outlineSize=0
      outlineColor=<rgba 0 0 0 1>
     }
    }
   }
   slot {class=StackBoxSlot, properties {verticalAlignment=VAlign_Center, padding=<ltrb 10 0 0 0>, size {sizeRule=Fill, value=1}}}
  }
  Image_65 {
   widgetClassPath=&Image
   bIsVariable=true
   parent=VerticalBox_0
   widget=Image_65
   properties {brush {
     bIsDynamicallyLoaded=false
     drawAs=Image
     imageType=NoImage
     mirroring=NoMirror
     resourceName=None
     resourceObject=null
     tiling=NoTile
     imageSize=<32 2>
     margin=<ltrb 0 0 0 0>
     outlineSettings {
      bUseBrushTransparency=false
      roundingType=HalfHeightRadius
      width=0
      color {colorUseRule=UseColor_Specified, specifiedColor=<rgba 0 0 0 0>}
      cornerRadii=<0 0 0 0>
     }
     tintColor {colorUseRule=UseColor_Specified, specifiedColor=<rgba 0 0 0 1>}
     uVRegion {bIsValid=false, max=<0 0>, min=<0 0>}
    }}
   slot {class=VerticalBoxSlot}
  }
 ]
}
