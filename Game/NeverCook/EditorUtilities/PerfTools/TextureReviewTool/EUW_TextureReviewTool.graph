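# Editor utility widget that reviews selected Texture2D assets for empty channels and lists those flagged.
# The run button loads the textures, runs them through the GART empty-texture report, then filters results by severity and two toggles into list items.
# Clicking an item syncs the Content Browser; the list exports to a named collection or, via embedded Python, to CSV.
# Toggles and collection name persist through the base widget's SaveGame hooks.
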EditorUtilityWidgetBlueprint /Game/NeverCook/EditorUtilities/PerfTools/TextureReviewTool/EUW_TextureReviewTool
graphs [
 AddListToCollection {nodes [
   AddListToCollection() {is=title K2Node_FunctionEntry}
   n1 = TextIsEmpty(InText="GetText (Text Box)"!(self:Text Box Object Reference=Get editTextCollectionName() -> (editTextCollectionName:Editor Utility Editable Text Box Object Reference)) -> (ReturnValue:Text))
   n2 = GetNumItems!(self:List View Object Reference=Get textureList() -> (textureList:Editor Utility List View Object Reference)) -> (ReturnValue:int=0)
   > Branch(Condition=AND Boolean(A="integer > integer"(A=@n2), B=NOT Boolean(A=@n1)))
   > n3 = Set l_items(l_items:Array of Object References=GetListItems!(self:List View Object Reference=Get textureList()) -> (ReturnValue:Array of Object References)) -> (Output_Get:Array of Object References)
   > n4 = For Each Loop(Array:Array of Object References=@n3) -> (Array Element:Object Reference)
   n5 = Get editTextCollectionName()
   SaveAssetsToCollection!(
    execute=@n4.Completed
    Objects:Array of Object References
    AssetData:Array of Asset Data Structures=Get l_assetDatas() -> (l_assetDatas:Array of Asset Data Structures)
    Collection:String="To String (Text)"(InText="GetText (Text Box)"(self=@n5))
    CollectionType:int=0
    Overwrite?:bool=true
    create?:bool=true
    __WorldContext:Object Reference
   ) -> (then, Success?:bool)
   n6 = Cast To TextureReviewToolData!(execute=@n4.LoopBody, Object=@n4) -> (then, CastFailed, AsTexture Review Tool Data:Texture Review Tool Data Object Reference) {is=title K2Node_DynamicCast}
   > Add(
    TargetArray:Array of Asset Data Structures=Get l_assetDatas()
    NewItem:"Asset Data Structure (by ref)"=Break Struct GART Asset Report List!(Struct_GART_AssetReportList:"Struct GART Asset Report List Structure (by ref)"=Get GartData!(self:Texture Review Tool Data Object Reference=@n6) -> (GartData:Struct GART Asset Report List Structure)) -> (
     ReportDatas_11_#1:Array of Struct GART Report Output Data Structures
     AssetData_7_#2:Asset Data Structure
     Object_10_#3:Object Reference
    ) {is=title K2Node_BreakStruct}.AssetData_7_#2
   ) {is=title K2Node_CallArrayFunction}
  ]}
 EventGraph {nodes [
   n1 = SpawnAndRegisterTabAndGetID!(
    execute
    self:Editor Utility Subsystem Object Reference=Get EditorUtilitySubsystem()
    InBlueprint:Editor Utility Widget Blueprint Object Reference=/Game/NeverCook/EditorUtilities/GenericAssetReviewTool/Widgets/EUW_GART_List_Main
   ) -> (then, NewTabID:Name=None, ReturnValue:Editor Utility Widget Object Reference)
   > n2 = Cast To EUW_GART_List_Main!(execute, Object=@n1.ReturnValue) -> (then, CastFailed, AsEUW GART List Main:EUW GART List Main Object Reference) {is=title K2Node_DynamicCast}
   n3 = GetEditorWorld(execute=@n4.then, self=Get UnrealEditorSubsystem())
   n4 = GetSelectedAssetData(execute=@n7.then)
   n5 = RunAssetsThroughFilters!(
    execute=@n18.then_1
    AssetDatas:Array of Asset Data Structures=@n4
    ReportClasses:Array of GART Report Base Soft Class References=Get ReportClasses() -> (ReportClasses:Array of GART Report Base Soft Class References)
    AllowLoading:bool=true
    __WorldContext:Object Reference=@n3
   ) -> (then, ReportDataMap:Map of Names to Struct GART Asset Report List Structures)
   SetAssetReportData!(
    execute=@n2.then
    self:EUW GART List Main Object Reference=@n2
    ReportData:Map of Names to Struct GART Asset Report List Structures=@n5
    UpdateList:bool=true
   ) -> (then)
   "On Clicked (btnSelectedAssets)"!() -> (OutputDelegate, then, widget:User Widget Object Reference, event:Pointer Event Structure) {is=title K2Node_ComponentBoundEvent}
   n6 = Get textureList()
   > Clear Selection(self:List View Object Reference=@n6)
   > n7 = ClearListItems(self=@n6)
   n8 = "On Item Clicked (textureList)"!() -> (OutputDelegate, then, Item:Object Reference) {is=title K2Node_ComponentBoundEvent}
   > n9 = GetEditorProperty!(execute:Exec, Object:Object Reference=@n8.Item, PropertyName:Name=GartData) -> (then:Exec, PropertyValue:"Struct GART Asset Report List Structure (by ref)", ReturnValue:bool)
   > Branch(Condition=@n9.ReturnValue)
   n10 = Break Asset Data(AssetData=Break Struct GART Asset Report List(Struct_GART_AssetReportList=@n9).AssetData_7_#2)
   > SyncBrowserToObjects(AssetPaths=Make Array("[0]":String="To String (Name)"(InName=@n10)) -> (Array:Array of Strings))
   "On Property Changed (detailsView)"!() -> (OutputDelegate, then, PropertyName:Name=None) {is=title K2Node_ComponentBoundEvent}
   > SaveGame_Save(self:TA Base Editor Widget Object Reference)
   "On Text Changed (editTextCollectionName)"!() -> (OutputDelegate, then, Text:"Text (by ref)") {is=title K2Node_ComponentBoundEvent}
   > SaveGame_Save()
   > ValidateCollectionButton()
   "On Clicked (btnExportCollection)"!() -> (OutputDelegate, then, widget:User Widget Object Reference, event:Pointer Event Structure) {is=title K2Node_ComponentBoundEvent}
   > AddListToCollection() {is=title K2Node_CallFunction}
   "On Clicked (btnExportCSV)"!() -> (OutputDelegate, then, widget:User Widget Object Reference, event:Pointer Event Structure) {is=title K2Node_ComponentBoundEvent}
   > ExportListAsCSV()
   Event Construct()
   > Parent: Construct()
   > SetObject(self=Get detailsView() -> (detailsView:Details View Object Reference), NewObject=Self-Reference())
   n11 = Construct Object from Class(execute=@n15.then, Class=&/TextureReviewToolData.TextureReviewToolData_C) -> (ReturnValue:Texture Review Tool Data Object Reference)
   AddItem!(execute=@n20.then, self:List View Object Reference=Get textureList(), Item:Object Reference=@n11) -> (then)
   GetAssetRegistry()
   LoadAssets!(execute=@n18.then_0, self, Array:Array of Asset Data Structures=@n4) -> (then)
   n12 = Values!(execute=@n19.then, TargetMap:Map of Names to Struct GART Asset Report List Structures=@n5) -> (then, Values:Array of Struct GART Asset Report List Structures)
   > n13 = For Each Loop(Array:Array of Struct GART Asset Report List Structures=@n12) -> (Array Element:Struct GART Asset Report List Structure)
   > n14 = ValidateTexture!(execute, self, Struct_GART_AssetReportList:"Struct GART Asset Report List Structure (by ref)"=@n13) -> (then, ReturnValue:bool)
   > n15 = Branch(Condition=@n14)
   n16 = GetAssetRegistry()
   n17 = MakeTopLevelAssetPath!(PackageName:String=/Script/Engine.Texture2D, AssetName:String) -> (ReturnValue:Top Level Asset Path Structure)
   RunAssetsThroughFilter!(
    execute=@n3.then
    self:Asset Registry Interface=@n16
    AssetDataList:Array of Asset Data Structures=@n4
    Filter:"ARFilter Structure (by ref)"=MakeARFilter!(
     PackageNames:Array of Names
     PackagePaths:Array of Names
     SoftObjectPaths:Array of Soft Object Path Structures
     ClassPaths:Array of Top Level Asset Path Structures=Make Array("[0]":Top Level Asset Path Structure=@n17) -> (Array:Array of Top Level Asset Path Structures)
     RecursiveClassPathsExclusionSet:Set of Top Level Asset Path Structures
     ClassNames:Array of Names
     RecursiveClassesExclusionSet:Set of Names
     bRecursivePaths:bool=false
     bRecursiveClasses:bool=false
     bIncludeOnlyOnDiskAssets:bool=false
    ) -> (ReturnValue:ARFilter Structure)
   ) -> (then)
   > n18 = Sequence()
   n19 = Keys!(execute=@n5.then, TargetMap:Map of Names to Struct GART Asset Report List Structures=@n5) -> (then, Keys:Array of Names)
   n20 = Set GartData!(execute=@n11.then, GartData:Struct GART Asset Report List Structure=@n13, self:Texture Review Tool Data Object Reference=@n11) -> (then, Output_Get:Struct GART Asset Report List Structure)
  ]}
 ExportListAsCSV {nodes [
   ExportListAsCSV() {is=title K2Node_FunctionEntry}
   n1 = TextIsEmpty(InText="GetText (Text Box)"(self=Get editTextCollectionName()))
   n2 = GetNumItems(self=Get textureList()) -> (ReturnValue=0)
   > Branch(Condition=AND Boolean(A="integer > integer"(A=@n2), B=NOT Boolean(A=@n1)))
   > n3 = Set l_items(l_items=GetListItems(self=Get textureList()))
   > n4 = For Each Loop(Array:Array of Object References=@n3) -> (Array Element:Object Reference)
   Execute Python Script!(
    execute=@n4.Completed
    PythonScript="import unreal\r\nimport csv\r\nimport os\r\nimport subprocess\r\n\r\ncsv_file_path = unreal.Paths.project_saved_dir() + \"TextureReviewToolExport.csv\"\r\ncsv_columns = [\"asset\", \"package name\", \"desc\"]\r\n\r\ndef generateCsvFromFinalDict(asset_datas, descs):\r\n    \r\n    if len(asset_datas) < 1: \r\n        unreal.log_error(\"Tried to export csv for Texture Review Tool but there were no assets to export.\")\r\n        return\r\n\r\n    final_dict = []\r\n    for i in range(len(asset_datas)):\r\n        asset_name = str(asset_datas[i].asset_name)\r\n        package_name = str(asset_datas[i].package_name)\r\n        desc = str(descs[i])\r\n        final_dict.append( {\"asset\":asset_name,\r\n                      \"package name\":package_name,\r\n                      \"desc\":desc} )\r\n        \r\n    try:\r\n        with open(csv_file_path, mode='w', newline='') as csvfile:\r\n            writer = csv.DictWriter(csvfile, fieldnames=csv_columns)\r\n            writer.writeheader()\r\n            writer.writerows(final_dict)\r\n            ## for data in final_dict:\r\n            ##     writer.writerow(data)\r\n        \r\n        openCsvLocation()\r\n    except IOError:\r\n        print(IOError)\r\n\r\ndef openCsvLocation():\r\n    csv_file_path = unreal.Paths.project_saved_dir() + \"AssetMapRefCount.csv\"\r\n    file_path = unreal.Paths.convert_relative_path_to_full(csv_file_path)\r\n    folder_path = unreal.Paths.get_path(file_path)\r\n    \r\n    real_path = os.path.realpath(folder_path)\r\n    subprocess.Popen(['explorer', real_path])\r\n\r\ngenerateCsvFromFinalDict(in_asset_datas, in_descs)"
    in_asset_datas:Array of Asset Data Structures=Get l_assetDatas()
    in_descs:Array of Strings=Get l_descs() -> (l_descs:Array of Strings)
   ) -> (then, ReturnValue) {inputs [in_asset_datas, in_descs]}
   n5 = Cast To TextureReviewToolData(execute=@n4.LoopBody, Object=@n4)
   n6 = Break Struct GART Asset Report List(Struct_GART_AssetReportList=Get GartData(self=@n5))
   > Add(TargetArray=Get l_assetDatas(), NewItem=@n6.AssetData_7_#2)
   n7 = "Get (a copy)"(Array:Array of Struct GART Report Output Data Structures=@n6) -> (Output:Struct GART Report Output Data Structure)
   > Add(
    TargetArray:Array of Strings=Get l_descs()
    NewItem:"String (by ref)"=Break Struct GART Report Data!(Struct_GART_ReportData:"Struct GART Report Data Structure (by ref)"=Break Struct GART Report Output Data!(Struct_GART_ReportOutputData:"Struct GART Report Output Data Structure (by ref)"=@n7) -> (ReportData_46_#4:Struct GART Report Data Structure) {is=title K2Node_BreakStruct}) -> (
     DisplayTitle_20_#5:String
     DisplayDescShort_29_#6:String
     DisplayDescLong_30_#7:String
     DisplayColorOverride_40_#8:Linear Color Structure
     Tags_25_#9:Array of Names
     Severity_46_#4:Enum_GART_ReportSeverity Enum=~
    ) {is=title K2Node_BreakStruct}.DisplayDescLong_30_#7
   )
  ]}
 LoadAssets {nodes [
   n1 = LoadAssets!() -> (then, Array:Array of Asset Data Structures) {is=title K2Node_FunctionEntry}
   > n2 = For Each Loop(Array:Array of Asset Data Structures=@n1) -> (Array Element:Asset Data Structure)
   > Set l_load(l_load:Object Reference=GetAsset(InAssetData=@n2)) -> (Output_Get:Object Reference)
  ]}
 SaveGame_PostLoad {nodes [
   n1 = SaveGame_PostLoad!() -> (then, saveGameExisted:bool, saveGame_Object:Save Game Object Reference) {is=title K2Node_FunctionEntry}
   > Branch(Condition=@n1)
   > n2 = Sequence()
   > n3 = Cast To TA_Base_SaveGame!(execute, Object=@n1.saveGame_Object) -> (then, CastFailed, AsTA Base Save Game:TA Base Save Game Object Reference) {is=title K2Node_DynamicCast}
   n4 = Get booleans!(self:TA Base Save Game Object Reference=@n3) -> (booleans:Array of Booleans)
   > Set ReportCompletelyEmptyOnly(ReportCompletelyEmptyOnly:bool="Get (a copy)"(Array:Array of Booleans=@n4) -> (Output:bool)) -> (Output_Get:bool)
   > Set AlwaysReportEmptyAlpha(AlwaysReportEmptyAlpha:bool="Get (a copy)"(Array=@n4, Dimension 1=1)) -> (Output_Get:bool)
   n5 = Get strings!(self:TA Base Save Game Object Reference=@n3) -> (strings:Array of Strings)
   > "SetText (Text Box)"!(
    execute
    self:Text Box Object Reference=Get editTextCollectionName()
    InText:Text="To Text (String)"(InString="Get (a copy)"(Array:Array of Strings=@n5) -> (Output:String))
   ) -> (then)
   ValidateCollectionButton(execute=@n2.then_1)
   Return Node!(execute, " ":bool=false)
  ]}
 SaveGame_PrepareToSave {nodes [
   n1 = SaveGame_PrepareToSave!() -> (then, SaveGameObject:Save Game Object Reference) {is=title K2Node_FunctionEntry}
   > n2 = Cast To TA_Base_SaveGame(Object=@n1)
   > Set booleans!(
    execute
    booleans:Array of Booleans=Make Array!(
     "[0]":bool=Get ReportCompletelyEmptyOnly() -> (ReportCompletelyEmptyOnly:bool)
     "[1]":bool=Get AlwaysReportEmptyAlpha() -> (AlwaysReportEmptyAlpha:bool)
    ) -> (Array:Array of Booleans)
    self:TA Base Save Game Object Reference=@n2
   ) -> (then, Output_Get:Array of Booleans)
   n3 = "GetText (Text Box)"(self=Get editTextCollectionName())
   > Set strings!(
    execute
    strings:Array of Strings=Make Array!("[0]":String="To String (Text)"(InText=@n3)) -> (Array:Array of Strings)
    self:TA Base Save Game Object Reference=@n2
   ) -> (then, Output_Get:Array of Strings)
   Return Node(" "=false)
  ]}
 ValidateCollectionButton {nodes [
   ValidateCollectionButton() {is=title K2Node_FunctionEntry}
   n1 = TextIsEmpty(InText="GetText (Text Box)"(self=Get editTextCollectionName()))
   n2 = Get btnExportCollection() -> (btnExportCollection:TA Generic 01 Button Object Reference)
   > SetIsEnabled(self=@n2, bInIsEnabled=NOT Boolean(A=@n1))
   > Set ToolTipText!(
    execute
    ToolTipText:Text=Select(
     Option 0:Text="LOCTEXT(\"Add these assets to a collection named whatever is input on the right --->\")"
     Option 1:Text="LOCTEXT(\"MUST HAVE A VALID NAME FOR THE COLLECTION INPUT ON THE RIGHT --->\")"
     Index=@n1
    ) -> (ReturnValue:Text)
    self:Widget Object Reference=@n2
   ) -> (then, Output_Get:Text)
  ]}
 ValidateTexture {nodes [
   n1 = ValidateTexture!() -> (then, Struct_GART_AssetReportList:"Struct GART Asset Report List Structure (by ref)") {is=title K2Node_FunctionEntry}
   n2 = Get ReportCompletelyEmptyOnly()
   n3 = Break Struct GART Asset Report List(Struct_GART_AssetReportList=@n1)
   n4 = Break Struct GART Report Data(Struct_GART_ReportData=Break Struct GART Report Output Data(Struct_GART_ReportOutputData="Get (a copy)"(Array:Array of Struct GART Report Output Data Structures=@n3) -> (Output:Struct GART Report Output Data Structure))) -> (Severity_46_#4=~)
   n5 = Contains(SearchIn=@n4.DisplayDescLong_30_#7, Substring=Alpha)
   n6 = Get AlwaysReportEmptyAlpha()
   > Return Node!(
    execute
    ReturnValue:bool=OR Boolean!(
     A="Equal (Enum)"(A:Enum_GART_ReportSeverity Enum=@n4.Severity_46_#4, B:Enum_GART_ReportSeverity Enum=NewEnumerator2)
     B=NOT Boolean(A=@n2)
     C:bool=AND Boolean(A=@n6, B=@n5)
    ) -> (ReturnValue)
   )
  ]}
]
variables [
 AlwaysReportEmptyAlpha {type=boolean}
 ReportClasses {
  type="array<Class>"
  default [/Game/NeverCook/EditorUtilities/PerfTools/TextureReviewTool/Report/Report_TextureEmpty_ReviewTool.Report_TextureEmpty_ReviewTool_C]
 }
 ReportCompletelyEmptyOnly {type=boolean}
]
widgets {
 info {
  parentClass=/Game/Blueprints/Blutilities/ForTechArt/Widgets/TA_Base_EditorWidget.TA_Base_EditorWidget_C
  rootWidgetClass=&StackBox
  widgetCount=10
 }
 widgets [
  StackBox_0 {widgetClassPath=&StackBox, widget=StackBox_0, properties {orientation=Orient_Vertical}}
  detailsView {
   widgetClassPath=&DetailsView
   parent=StackBox_0
   widget=detailsView
   properties {categoriesToShow [nocategory]}
   slot {class=StackBoxSlot}
  }
  btnSelectedAssets {
   widgetClassPath=&TA_Generic01_Button
   parent=StackBox_0
   widget=btnSelectedAssets
   properties {label=Run on Selected Assets}
   slot {class=StackBoxSlot, properties {padding=<ltrb 0 10 0 0>}}
  }
  Border_1 {
   widgetClassPath=&Border
   parent=StackBox_0
   widget=Border_1
   properties {brushColor=<rgba 0 0 0 1>, padding=<ltrb 10 10 10 10>}
   slot {class=StackBoxSlot, properties {size {sizeRule=Fill, value=1}}}
  }
  Border_0 {
   widgetClassPath=&Border
   parent=Border_1
   widget=Border_0
   properties {brushColor=<rgba 0.01 0.01 0.01 1>, padding=<ltrb 5 5 5 5>}
   slot {class=BorderSlot, properties {padding=<ltrb 10 10 10 10>}}
  }
  textureList {
   widgetClassPath=/Script/Blutility.EditorUtilityListView
   bIsVariable=true
   parent=Border_0
   widget=textureList
   properties {
    entryWidgetClass=&/EUW_TextureReviewToolItem.EUW_TextureReviewToolItem_C
    verticalEntrySpacing=5
   }
   slot {class=BorderSlot, properties {padding=<ltrb 5 5 5 5>}}
  }
  HorizontalBox_0 {
   widgetClassPath=&HorizontalBox
   parent=StackBox_0
   widget=HorizontalBox_0
   slot {class=StackBoxSlot, properties {padding=<ltrb 0 10 0 0>}}
  }
  btnExportCollection {
   widgetClassPath=&TA_Generic01_Button
   parent=HorizontalBox_0
   widget=btnExportCollection
   properties {label=Add Assets To Collection, buttonColor_Default=<rgba 0.1775 0.4896 0.1701 1>}
   slot {class=HorizontalBoxSlot, properties {size {sizeRule=Fill, value=3}}}
  }
  editTextCollectionName {
   widgetClassPath=&EditorUtilityEditableTextBox
   bIsVariable=true
   parent=HorizontalBox_0
   widget=editTextCollectionName
   properties {text=111_TexturesWithEmptyChannels}
   slot {class=HorizontalBoxSlot, properties {size {sizeRule=Fill, value=1}}}
  }
  btnExportCSV {
   widgetClassPath=&TA_Generic01_Button
   parent=StackBox_0
   widget=btnExportCSV
   properties {label=Export List to CSV, buttonColor_Default=<rgba 0.1775 0.4896 0.1701 1>}
   slot {class=StackBoxSlot, properties {padding=<ltrb 0 10 0 0>}}
  }
 ]
}
